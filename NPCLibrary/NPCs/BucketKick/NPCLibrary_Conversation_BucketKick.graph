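# Conversation database for the BucketKick NPC, defining the greeting speech shown when a player starts talking to them.
# The entry point leads to one speech node that picks a line from contextual messages: a first-ever meeting greeting, lines gated on per-match conversation count, and reactions to specific worn outfits.
# Priority curves rank the first-meeting and outfit lines, with a default line as fallback.
# The speech then branches through link nodes to shared quests, basic services and BucketKick's own dialogue entry.
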ConversationDatabase /NPCLibrary/NPCs/BucketKick/NPCLibrary_Conversation_BucketKick
entry_points [{tag=Conversation.Entry.BucketKick, destinations [#1]}]
nodes [
 ConversationEntryPointNode_7 {
  id=#1
  class=ConversationEntryPointNode
  outputs [#2]
  properties {entryTag {tagName=Conversation.Entry.BucketKick}}
 }
 Speech_C_4 {
  id=#2
  class=Speech_C
  outputs [#3, #4, #5]
  properties {
   nodeName=""
   generalConfig {
    defaultMessage=I will stop at nothing to keep the nightmares at bay.
    contextualMessages [
     {
      message=You've found me. Greetings.
      contextRequirements [{requirement {class=FortControllerRequirement_IsFirstEverConversationWithNPC}}]
      priority {curve {rowName=Tandem.Speech.Priority.FirstEverInteraction}}
     }
     {
      message=I will stop at nothing to keep the nightmares at bay.
      contextRequirements [{requirement {class=FortControllerRequirement_CheckNPCMatchConversationCount, properties {nPCUniqueIdTag {tagName=AISpawnerData.Type.Tandem.BucketKick}}}}]
      priority {curve {rowName=None}}
     }
     {
      message=I am at war with the hideous beasts that lurk in the shadows...
      contextRequirements [{requirement {class=FortControllerRequirement_CheckNPCMatchConversationCount, properties {nPCUniqueIdTag {tagName=AISpawnerData.Type.Tandem.BucketKick}}}}]
      priority {curve {rowName=None}}
     }
     {
      message=Nightmares?? *laugh* I don't fear them. They fear ME.
      contextRequirements [{requirement {class=FortControllerRequirement_CheckNPCMatchConversationCount, properties {nPCUniqueIdTag {tagName=AISpawnerData.Type.Tandem.BucketKick}}}}]
      priority {curve {rowName=None}}
     }
     {
      message=Hmm... How do I get up there?
      contextRequirements [{requirement {class=FortControllerRequirement_CheckNPCMatchConversationCount, properties {nPCUniqueIdTag {tagName=AISpawnerData.Type.Tandem.BucketKick}}}}]
      priority {curve {rowName=None}}
     }
     {
      message=A banana skeleton? Now I really have seen everything.
      contextRequirements [{requirement {
         class=FortControllerRequirement_HasCID
         properties {allowedCIds [/BRCosmetics/Athena/Items/Cosmetics/Characters/CID_602_Athena_Commando_M_NanaSplit]}
        }}]
      priority {curve {rowName=Tandem.Speech.Priority.OutfitReaction}}
     }
     {
      message=Ugh. Keep your Bitey llamacorn away from me.
      contextRequirements [{requirement {class=FortControllerRequirement_HasCID, properties {allowedCIds [/BRCosmetics/Athena/Items/Cosmetics/Characters/Character_ShiverFlame_Ace]}}}]
      priority {curve {rowName=Tandem.Speech.Priority.OutfitReaction}}
     }
    ]
   }
  }
 }
 ConversationLinkNode_2 {
  id=#5
  class=ConversationLinkNode
  properties {remoteEntryTag {tagName=Conversation.Entry.BucketKick.Dialogue}}
 }
 ConversationLinkNode_0 {id=#4, class=ConversationLinkNode, properties {remoteEntryTag {tagName=Conversation.Entry.BasicServices}}}
 ConversationLinkNode_1 {id=#3, class=ConversationLinkNode, properties {remoteEntryTag {tagName=Conversation.Entry.NPCLibrary.Quests}}}
]
